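# Character material for a katana scabbard with animated emissive, layered on the shared character PreFX, skin FX and PostFX functions.
# A green-dominance mask on emissive color blends detail glow with a blade glow pulsed by a scrolling sine, labeled for idle flow and charge-gain blink.
# A parameter-positioned, bounding-box sweep band tinted by base color is added.
# Time of day scales the emissive before cinematic character post FX.
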Material /KatanaGameplay/Materials/M_Sharp_KatanaScabbard
graph {
 properties {
  bAutomaticallySetUsageInEditor=false
  bUseMaterialAttributes=true
  bUsedWithClothing=true
  bUsedWithMeshDeformer=true
  bUsedWithMorphTargets=true
  bUsedWithNanite=true
  bUsedWithSkeletalMesh=true
  pixelDepthOffsetMode=PDOM_Legacy
  refractionMethod=RM_IndexOfRefraction
  shadingModel=MSM_FromMaterialExpression
  subsurfaceProfile=&SS_HeroSkin_02
  usedShadingModels=Default Lit | Subsurface | Clear Coat | Subsurface Profile | Cloth | Eye
 }
 nodes [
  Frac(None=Time())
  n1 = MaterialFunctionCall() {materialFunction=&BaseFN_CharacterFunction_PreFX}
  n2 = MaterialFunctionCall(
   AO=@n1.AO
   Distance Blend=@n1.AO
   Fresnel=@n1.AO
   Gmap=@n1.AO
   Pre Skinned Local Position=@n1.AO
   Material=MaterialFunctionCall(
    PreskinnedPosition=@n1.Pre skinned Local Position
    Gmap=@n1.Pre skinned Local Position
    MaterialAttributes=@n1.Pre skinned Local Position
    AO=@n1.Pre skinned Local Position
   ) {materialFunction=&CharacterSkinFX}.Result
  ) {materialFunction=&BaseFN_CharacterFunction_PostFX}
  n3 = GetMaterialAttributes(None=@n2.Result) {attributeGetTypes [#1, #2]}
  n4 = ScalarParameter() {defaultValue=1, group=Emissive, parameterName=EmissiveIntensity Details}
  n5 = ComponentMask(None=TextureCoordinate()) {r=true}
  n6 = Multiply(A=Time()) {constB=0.5}
  n7 = ConstantBiasScale(None=Sine(None=Subtract(A=@n5, B=@n6)) {period=0.5}) {bias=1.25, scale=0.8}
  n8 = ScalarParameter() {defaultValue=2, group=Emissive, parameterName=EmissiveIntensity Blade Peak}
  n9 = ComponentMask(None=@n3.Emissive Color) {g=true}
  n10 = Length(None=@n3.Emissive Color) {is=class MaterialExpressionLength}
  n11 = Floor(None=Add(A=Divide(A=@n9, B=@n10)) {constB=0.5})
  n12 = Saturate(None=@n11)
  n13 = Multiply(
   A=Multiply(A=@n3.Emissive Color, B=ScalarParameter() {defaultValue=1, group=Emissive, parameterName=EmissiveIntensity Blade Base})
   B=Add(A=Multiply(A=@n8, B=@n12) {constB=2})
  )
  n14 = Subtract(A=@n11, B=Subtract(A=ScalarParameter() {defaultValue=1, group=Emissive, parameterName=emissive mult}))
  n15 = ScalarParameter() {defaultValue=4, group=Emissive, parameterName=EmissiveSweepIntensity}
  n16 = MaterialFunctionCall() {materialFunction=&BoundingBoxBased_0-1_UVW}
  n17 = ScalarParameter() {defaultValue=1, group=Emissive, parameterName=EmissiveSweep}
  n18 = ConstantBiasScale(None=Subtract(A=Add(A=@n16.B) {constB=0.5}, B=Add(A=@n17) {constB=0.3})) {bias=-0.5, scale=2}
  n19 = Power(Base=OneMinus(None=Abs(None=@n18))) {constExponent=8}
  n20 = Add(
   A=LinearInterpolate(A=Multiply(A=@n3.Emissive Color, B=@n4), B=Multiply(A=@n7, B=@n13), Alpha=Saturate(None=@n14)) {desc="blink on charge gained && flowing idle"}
   B=Multiply(A=@n3.Base Color, B=Multiply(A=@n15, B=@n19) {constB=4}) {desc=sweep}
  )
  n21 = Constant() {r=1.5}
  n22 = MaterialFunctionCall(Value_Morning=@n21, Value_Day=Constant() {r=2}, Value_Evening=@n21, Value_Night=Constant() {r=1}) {materialFunction=&Scalar_Time_of_Day}
  MaterialFunctionCall(In=SetMaterialAttributes(MaterialAttributes=@n2.Result, Emissive Color=Multiply(A=@n20, B=@n22.Result)) {attributeSetTypes [#1]}) {materialFunction=&CINE_MF_CharPostFX}
 ]
}
