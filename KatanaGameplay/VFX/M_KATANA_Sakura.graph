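# M_KATANA_Sakura is an unlit, translucent, two-sided material for Niagara sprites that renders cherry blossom petals in the Katana VFX.
# Color combines the petal texture, sharpened by a gradient power, with particle color and a brightness multiplier chosen by time of day through Color_Time_of_Day.
# A second chain multiplies texture alpha by a powered near-camera fade, so petals fade out close to the camera.
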Material /KatanaGameplay/VFX/M_KATANA_Sakura
graph {
 properties {
  bUsedWithNiagaraSprites=true
  blendMode=BLEND_Translucent
  pixelDepthOffsetMode=PDOM_Legacy
  shadingModel=MSM_Unlit
  twoSided=true
  usedShadingModels=Unlit
 }
 nodes [
  n1 = ScalarParameter() {defaultValue=0.5, parameterName=HDRMult_3}
  n2 = ScalarParameter() {defaultValue=1, parameterName=HDRMult_2}
  n3 = ScalarParameter() {defaultValue=0.5, parameterName=HDRMult_1}
  n4 = ScalarParameter() {defaultValue=0.35, parameterName=HDRMult}
  n5 = TextureSample() {texture=/Game/Effects/Fort_Effects/Textures/Leaf/T_Cherry_Petals_64_01}
  Multiply(
   A=Multiply(
    A=MaterialFunctionCall(Color_Morning=@n1, Color_Day=@n2, Color_Evening=@n3, Color_Night=@n4) {materialFunction=&Color_Time_of_Day}.Result
    B=ParticleColor().RGB
   )
   B=Power(Base=@n5.RGB, Exp=ScalarParameter() {defaultValue=1.5, parameterName=Gradient Pow})
  )
  n6 = ScalarParameter() {defaultValue=200, parameterName=Near Camera Fade}
  Multiply(
   A=Power(Base=MaterialFunctionCall(Distance=@n6) {materialFunction=&Near_Camera_Fade}.Result) {constExponent=3.5}
   B=@n5.A
  )
 ]
}
