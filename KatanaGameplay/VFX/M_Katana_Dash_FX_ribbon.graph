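# Unlit, two-sided translucent material for katana dash FX on Niagara ribbons and mesh particles.
# Color samples a katana texture through noise-normal UV distortion, then applies time-of-day glow and particle color.
# Opacity combines panning noise, distorted mask, alpha gradient and edge masks, with switchable dynamic-parameter clipout and a blendable liquid-trail alternative.
# Near-camera fade, particle alpha and optional depth fade shape the final opacity.
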Material /KatanaGameplay/VFX/M_Katana_Dash_FX_ribbon
graph {
 properties {
  bUsedWithNiagaraMeshParticles=true
  bUsedWithNiagaraRibbons=true
  blendMode=BLEND_Translucent
  pixelDepthOffsetMode=PDOM_Legacy
  shadingModel=MSM_Unlit
  twoSided=true
  usedShadingModels=Unlit
 }
 nodes [
  n1 = TextureCoordinate()
  MaterialFunctionCall(XY=@n1) {materialFunction=&Swizzle}
  n2 = Multiply(
   A=NamedRerouteUsage()
   B=MaterialFunctionCall(
    Value_Morning=ScalarParameter() {defaultValue=0.5, group=Color, parameterName=Glow_Value_Morning}
    Value_Day=ScalarParameter() {defaultValue=1, group=Color, parameterName=Glow_Value_Day}
    Value_Evening=ScalarParameter() {defaultValue=0.5, group=Color, parameterName=Glow_Value_Evening}
    Value_Night=ScalarParameter() {defaultValue=0.5, group=Color, parameterName=Glow_Value_Night}
   ) {materialFunction=&Scalar_Time_of_Day}.Result
  )
  MaterialFunctionCall(Emissive=Multiply(A=ParticleColor().RGB, B=@n2)) {materialFunction=&MF_FX_EmissiveAdjustments}
  n3 = ScalarParameter() {defaultValue=0.15, parameterName=UVPos}
  Multiply(A=Time(), B=@n3)
  n4 = DynamicParameter() {defaultValue=<rgba 1 1 0 1>, paramNames [rand offset, Clip, Random, Param4]}
  n5 = TextureCoordinate() {vTiling=0.2}
  n6 = VectorParameter() {parameterName=ColorDistortUV, defaultValue=<rgba 0.05 0.05 1 1>}
  n7 = ComponentMask(None=@n6.A) {a=true, b=true}
  n8 = LinearInterpolate(
   A=TextureCoordinate()
   B=TextureCoordinate() {coordinateIndex=1}
   Alpha=ScalarParameter() {defaultValue=1, group=Controls, parameterName=Use UV1}
  )
  n9 = ComponentMask(None=VectorParameter() {group=Cube Highlight, parameterName=Noise Normal UV Scale, defaultValue=<rgba 1 1 0 0>}.RGB) {g=true, r=true}
  n10 = Time() {bOverride_Period=true, period=64}
  n11 = DynamicParameter() {defaultValue=<rgba 1 1 0 1>, paramNames [rand offset, Clip, Random, Param4]}
  n12 = VectorParameter() {group=Cube Highlight, parameterName=Noise Normal Panner Speed, defaultValue=<rgba -0.3 -0.01 0 0>}
  n13 = TextureSampleParameter2D(UVs=Panner(Coordinate=Multiply(A=@n8, B=@n9), Time=Add(A=@n10, B=@n11.rand offset), Speed=ComponentMask(None=@n12.RGB) {g=true, r=true}) {speedX=-0.3, speedY=-0.01}) {
   group=Cube Highlight
   parameterName=Noise Normal Texture
   samplerSource=SSM_Wrap_WorldGroupSettings
   samplerType=SAMPLERTYPE_Normal
   texture=/Game/Global/Textures/Default/Tileable_Patterns/2ndInfectedMapNormal
  }
  n14 = ComponentMask(None=@n13.RGB) {g=true, r=true}
  n15 = ComponentMask(None=@n6.RGB) {g=true, r=true}
  n16 = Add(A=@n3, B=Add(A=Multiply(A=@n5, B=@n7), B=Multiply(A=@n14, B=@n15) {constB=0.05}))
  NamedRerouteDeclaration(None=TextureSampleParameter2D(UVs=Add(A=@n4.rand offset, B=@n16)) {parameterName=Texture, texture=&/T_Katana_vivid_0}.RGB) {name=ColorA, nodeColor=<rgba 0.5703 1 0 1>}
  n17 = TextureCoordinate()
  n18 = TextureCoordinate() {coordinateIndex=1}
  n19 = ScalarParameter() {defaultValue=1, group=Controls, parameterName=Use UV1}
  n20 = VectorParameter() {group=Distortion, parameterName=Distortion Tile, defaultValue=<rgba 6 1 0 0>}
  n21 = Panner(
   Coordinate=Multiply(A=LinearInterpolate(A=@n17, B=@n18, Alpha=@n19), B=ComponentMask(None=@n20.RGB) {g=true, r=true})
   Time=Add(
    A=Time() {bOverride_Period=true, period=64}
    B=DynamicParameter() {defaultValue=<rgba 1 1 0 1>, paramNames [rand offset, Clip, Random, Param4]}.rand offset
   )
   Speed=ComponentMask(None=VectorParameter() {group=Distortion, parameterName=Distortion Panner Speed, defaultValue=<rgba -4.5 0.1 0 0>}.RGB) {g=true, r=true}
  ) {speedX=-1, speedY=0.1}
  n22 = ScalarParameter() {defaultValue=1, group=DynamicClip, parameterName=DynamicClipoutTileU}
  n23 = ScalarParameter() {defaultValue=1, group=DynamicClip, parameterName=DynamicClipoutTileV}
  n24 = TextureSampleParameter2D(UVs=Multiply(A=@n21, B=AppendVector(A=@n22, B=@n23))) {
   group=DynamicClip
   parameterName=DynamicClipoutTexture
   samplerType=SAMPLERTYPE_LinearColor
   texture=/Game/Characters/Player/Male/Medium/Bodies/M_MED_SRDrift/Textures/T_ElectricNoise
  }
  n25 = ChannelMaskParameter(None=@n24.RGBA) {group=DynamicClip, maskChannel=Green, parameterName=DynamicClipoutTextureChannel, defaultValue=<rgba 0 1 0 0>}
  n26 = DynamicParameter() {defaultValue=<rgba 1 1 0 1>, paramNames [rand offset, Clip, Random, Param4]}
  n27 = Saturate(None=Add(A=@n25.R, B=LinearInterpolate(Alpha=@n26.Clip) {constA=-1}))
  n28 = MaterialFunctionCall!(
   UVs
   Mult=ScalarParameter() {defaultValue=2, group=Opacity, parameterName=EdgeMult}
   Exp=ScalarParameter() {defaultValue=1.6, group=Opacity, parameterName=EdgeExp}
  ) {materialFunction=/Game/Effects/Fort_Effects/MaterialFunctions/MF_SquareEdgeMask}
  n29 = OneMinus(None=TextureCoordinate())
  n30 = Multiply(A=Add(A=ComponentMask(None=@n29) {r=true}) {constB=-0.3}) {constB=0.5}
  n31 = Saturate(None=Add(A=Multiply(A=@n28.SquareMask), B=Saturate(None=@n30)))
  n32 = ScalarParameter() {group=Opacity, parameterName=EdgeMaskShape}
  n33 = MaterialFunctionCall(
   Mult=ScalarParameter() {defaultValue=8, group=Opacity, parameterName=FrontEdgeMult}
   Exp=ScalarParameter() {defaultValue=3, group=Opacity, parameterName=FrontEdgeExp}
  ) {materialFunction=&MF_SquareEdgeMask}
  n34 = OneMinus(None=MaterialFunctionCall(
    Mult=ScalarParameter() {defaultValue=8, group=Opacity, parameterName=EndEdgeMult}
    Exp=ScalarParameter() {defaultValue=3, group=Opacity, parameterName=EndEdgeExp}
   ) {materialFunction=&MF_SquareEdgeMask}.RightMask)
  n35 = ScalarParameter() {group=Opacity, parameterName=UseEndBreakUpLiquidShapeMask, sliderMax=1}
  n36 = LinearInterpolate(
   A=TextureCoordinate()
   B=TextureCoordinate() {coordinateIndex=1}
   Alpha=ScalarParameter() {defaultValue=1, group=Controls, parameterName=Use UV1}
  )
  n37 = ComponentMask(None=VectorParameter() {group=Noise, parameterName=Noise UV Scale, defaultValue=<rgba 1 0.1 0 0>}.RGB) {g=true, r=true}
  n38 = Time() {bOverride_Period=true, period=64}
  n39 = DynamicParameter() {defaultValue=<rgba 1 1 0 1>, paramNames [rand offset, Clip, Random, Param4]}
  n40 = VectorParameter() {group=Noise, parameterName=Noise Panner Speed, defaultValue=<rgba -0.8 0 0 0>}
  n41 = TextureSampleParameter2D(UVs=Panner(Coordinate=Multiply(A=@n36, B=@n37), Time=Add(A=@n38, B=@n39.rand offset), Speed=ComponentMask(None=@n40.RGB) {g=true, r=true}) {speedX=-0.1, speedY=-0.01}) {
   group=Noise
   parameterName=Noise Texture
   samplerSource=SSM_Wrap_WorldGroupSettings
   texture=/Game/Global/Textures/Default/Caustics/T_Fractal_01
  }
  n42 = ChannelMaskParameter(None=@n41.RGBA) {group=Noise, maskChannel=Green, parameterName=Noise Channel Mask, defaultValue=<rgba 0 1 0 0>}
  n43 = LinearInterpolate(B=@n42.R, Alpha=Add(A=OneMinus(None=@n33.LeftMask), B=LinearInterpolate(B=@n34, Alpha=@n35))) {constA=1}
  n44 = ScalarParameter() {defaultValue=0.15, group=Noise, parameterName=Noise Contrast}
  n45 = LinearInterpolate(
   A=TextureCoordinate()
   B=TextureCoordinate() {coordinateIndex=1}
   Alpha=ScalarParameter() {defaultValue=1, group=Controls, parameterName=Use UV1}
  )
  n46 = ComponentMask(None=VectorParameter() {group=Mask, parameterName=Mask UV Scale, defaultValue=<rgba 4 1 0 0>}.RGB) {g=true, r=true}
  n47 = TextureSampleParameter2D(UVs=@n21) {
   group=Distortion
   parameterName=Distortion Texture
   samplerSource=SSM_Wrap_WorldGroupSettings
   samplerType=SAMPLERTYPE_Grayscale
   texture=&LowResBlurredNoise
  }
  n48 = Multiply(
   A=LinearInterpolate(Alpha=ChannelMaskParameter(None=@n47.RGBA) {group=Distortion, maskChannel=Green, parameterName=Distortion Mask, defaultValue=<rgba 0 1 0 0>}.R) {constA=-1}
   B=ScalarParameter() {defaultValue=0.15, group=Distortion, parameterName=DistAmnt}
  )
  n49 = Panner(
   Coordinate=Add(A=Multiply(A=@n45, B=@n46), B=@n48)
   Time=Add(
    A=Time() {bOverride_Period=true, period=64}
    B=DynamicParameter() {defaultValue=<rgba 1 1 0 1>, paramNames [rand offset, Clip, Random, Param4]}.rand offset
   )
   Speed=ComponentMask(None=VectorParameter() {group=Mask, parameterName=Mask Panner Speed, defaultValue=<rgba -2.6 0 0 0>}.RGB) {g=true, r=true}
  ) {speedX=-0.8}
  n50 = ComponentMask(None=VectorParameter() {group=Mask, parameterName=Mask Texture Offset}.RGB) {g=true, r=true}
  n51 = ChannelMaskParameter(None=TextureSampleParameter2D(UVs=Add(A=@n49, B=@n50)) {group=Mask, parameterName=Mask Texture, texture=/Game/Effects/Fort_Effects/Textures/Smoke/T_Smoke_Trail_Soft_Chromatic}.RGBA) {group=Mask, maskChannel=Green, parameterName=Mask Channel Mask, defaultValue=<rgba 0 1 0 0>}
  n52 = ScalarParameter() {group=Mask, parameterName=Mask Contrast}
  n53 = TextureCoordinate()
  n54 = TextureCoordinate() {coordinateIndex=1}
  n55 = ScalarParameter() {defaultValue=1, parameterName=Use UV1}
  n56 = Add(
   A=Add(A=@n48, B=LinearInterpolate(A=@n53, B=@n54, Alpha=@n55))
   B=ComponentMask(None=VectorParameter() {group=Alpha Mask, parameterName=Alpha Mask UV Offset}.RGB) {g=true, r=true}
  )
  n57 = MaterialFunctionCall(Float2=@n56) {materialFunction=&BreakOutFloat2Components}
  n58 = LinearInterpolate(
   A=@n56
   B=AppendVector(A=OneMinus(None=@n57.R), B=OneMinus(None=@n57.G))
   Alpha=ScalarParameter() {group=Alpha Mask, parameterName=Invert Alpha Texture UVs}
  )
  n59 = Add(
   A=Multiply(
    A=MaterialFunctionCall(In=@n42.R, Contrast=@n44) {materialFunction=&CheapContrast}.Result
    B=MaterialFunctionCall(In=@n51.R, Contrast=@n52) {materialFunction=&CheapContrast}.Result
   )
   B=MaterialFunctionCall(
    In=TextureSampleParameter2D(UVs=@n58) {group=Alpha Mask, parameterName=Alpha Texture, texture=/Game/Global/Textures/Default/Masks_and_Gradients/gradient_stasis}.G
    Contrast=ScalarParameter() {defaultValue=0.5, group=Alpha Mask, parameterName=Alpha Texture Contrast}
   ) {materialFunction=&CheapContrast}.Result
  )
  n60 = ScalarParameter() {defaultValue=0.05, group=Opacity, parameterName=ClipRange}
  n61 = ScalarParameter() {defaultValue=0.9, group=Opacity, parameterName=ClipAmnt}
  n62 = Subtract(
   A=Multiply(A=LinearInterpolate(A=@n28.RightMask, B=@n31, Alpha=@n32), B=Multiply(A=@n43, B=@n59))
   B=LinearInterpolate(A=Multiply(A=@n60) {constB=-1}, Alpha=OneMinus(None=@n61) {desc=Inverse Alpha})
  )
  n63 = Multiply(A=Saturate(None=Divide(A=@n62, B=@n60))) {constB=0.9}
  n64 = StaticSwitchParameter(
   True=MaterialFunctionCall(
    Alpha=Multiply(A=@n27, B=@n63)
    Min=ScalarParameter() {defaultValue=0.05, group=DynamicClip, parameterName=DynamicClipMin}
    Max=ScalarParameter() {defaultValue=0.25, group=DynamicClip, parameterName=DynamicClipMax}
   ) {materialFunction=&SmoothStep}.Result
   False=@n63
  ) {group=DynamicClip, parameterName=DynamicClipAmount}
  n65 = MaterialFunctionCall(Float2=@n1) {materialFunction=&BreakOutFloat2Components}
  n66 = DynamicParameter() {defaultValue=<rgba 1 1 0 1>, paramNames [rand offset, Clip, Random, Param4]}
  n67 = Panner(Coordinate=AppendVector(A=Add(A=@n65.G, B=@n66.Random), B=@n65.R)) {speedX=0.25}
  n68 = TextureCoordinate()
  n69 = Add(
   A=TextureSample(UVs=@n67) {samplerSource=SSM_Wrap_WorldGroupSettings, texture=/Game/Effects/Fort_Effects/Textures/Tile/T_Plasma_BallEdge}.RGB
   B=OneMinus(None=ComponentMask(None=@n68) {r=true})
  )
  n70 = ScalarParameter() {defaultValue=4, group=Liquid Trail, parameterName=Liquid Trail Edge Contrast}
  n71 = TextureSampleParameter2D(UVs=@n67) {group=Alpha Mask, parameterName=Goop Edge Alpha Mask, texture=/Game/Global/Textures/Default/Masks_and_Gradients/T_Soft_Ramp_Vert_03}
  n72 = ScalarParameter() {defaultValue=1, desc="0 fully softens edges, 1 prevents softening", group=Liquid Trail, parameterName=SoftenEdge, sliderMax=1}
  n73 = Add(A=@n71.RGB, B=@n72)
  n74 = TextureCoordinate() {coordinateIndex=1, uTiling=0.5, vTiling=0.5}
  n75 = ComponentMask(None=VectorParameter() {group=Liquid Trail, parameterName=Liquid Trail Bubbles_UV_Scale, defaultValue=<rgba 1 1 0 0>}.RGB) {g=true, r=true}
  n76 = TextureSample(UVs=Panner(Coordinate=Multiply(A=@n74, B=@n75)) {speedX=0.0625}) {texture=/Game/Effects/Fort_Effects/Textures/Tile/T_LaserTrail_Dots}
  n77 = Add(
   A=TextureSample(UVs=@n74) {samplerSource=SSM_Wrap_WorldGroupSettings, texture=/Game/Effects/Fort_Effects/Textures/Tile/T_Plasma_BallCore}.R
   B=Multiply(A=Clamp(None=@n76.R), B=ScalarParameter() {defaultValue=0.25, group=Liquid Trail, parameterName=Liquid Trail Bubble Intensity})
  )
  n78 = Multiply(
   A=Multiply(A=Power(Base=@n69, Exp=@n70), B=Saturate(None=@n73))
   B=Multiply(A=Clamp(None=@n77), B=ScalarParameter() {defaultValue=2, group=Liquid Trail, parameterName=Liquid Trail Alpha Multiplier}) {constA=1, constB=10}
  )
  n79 = ScalarParameter() {group=Controls, parameterName="Use Goop or Liquid Trail "}
  n80 = Power(
   Base=MaterialFunctionCall(Distance=ScalarParameter() {defaultValue=90, group=Opacity, parameterName=NearFade}) {materialFunction=&Near_Camera_Fade}.Result
   Exp=ScalarParameter() {defaultValue=5, group=Opacity, parameterName=NearFadeExp}
  )
  n81 = Multiply(A=Multiply(A=LinearInterpolate(A=@n64, B=@n78, Alpha=@n79), B=Saturate(None=@n80)), B=ParticleColor().A)
  n82 = DepthFade(Opacity=@n81, FadeDistance=ScalarParameter() {defaultValue=20, group=DepthDade, parameterName=DepthFadeDistance, sortPriority=1})
  Reroute(None=LinearInterpolate(
    A=@n81
    B=ShadingPathSwitch(Default=@n82, Deferred=@n82, Mobile=@n81) {desc=Migrated from FeatureLevelSwitch}
    Alpha=ScalarParameter() {group=DepthDade, parameterName=UseDepthFade}
   ))
  Reroute(None=@n81)
  Reroute(None=@n2)
  n83 = CollectionParameter() {
   collection=&FortniteMaterialParameters
   parameterName=SunlightDirectionalVector
  }
  n84 = Multiply(
   A=LinearInterpolate(
    A=MaterialFunctionCall() {materialFunction=&CameraDirectionVector}.Result
    B=ComponentMask(None=@n83) {b=true, g=true, r=true}
    Alpha=ScalarParameter() {group=Liquid Color Scattering, parameterName=Use Sun Direction}
   )
   B=ScalarParameter() {defaultValue=1, group=Liquid Color Scattering, parameterName=Liquid_DirectionVectorFlip}
  )
  n85 = Normalize(VectorInput=Transform(None=ComponentMask(None=@n84) {b=true, g=true, r=true}) {transformSourceType=TRANSFORMSOURCE_World, transformType=TRANSFORM_Local})
  n86 = MaterialFunctionCall(Float3=Multiply(A=ComponentMask(None=@n85) {b=true, g=true, r=true}) {constB=0.4886}) {materialFunction=&BreakOutFloat3Components}
  n87 = DotProduct(
   A=MaterialFunctionCall(X=Constant() {r=0.2821}, Y=@n86.B, Z=@n86.B, A=@n86.B) {materialFunction=&MakeFloat4}.Result
   B=Constant4Vector() {constant=<rgba 2 0.5 0.5 0.5>}.RGBA
  )
  n88 = OneMinus(None=MaterialFunctionCall(
    Input=Saturate(None=@n87)
    Input Low=ScalarParameter() {group=Liquid Color Scattering, parameterName=Liquid_SHContrast_LowValue}
    Input High=ScalarParameter() {defaultValue=1, group=Liquid Color Scattering, parameterName=Liquid_SHContrast_HighValue}
    Target Low=Constant()
    Target High=Constant() {r=1}
   ) {materialFunction=&RemapValueRange}.Result)
  VertexInterpolator(VS=Power(Base=Saturate(None=@n88), Exp=ScalarParameter() {defaultValue=1, group=Liquid Color Scattering, parameterName=Liquid_SHPow}))
 ]
}
