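# Translucent, unlit, two-sided Niagara mesh-particle material for the head of the katana dash effect.
# Panning swirl and smooth noise form streaks with V-varying contrast, blending dark to light color beneath a V-gradient center glow and squared dust layers.
# Particle color tints the emissive after time-of-day HDR remapping.
# Opacity combines streaks and center gradient with particle alpha, inverse Fresnel, optional UV edge mask, depth fade and near-camera fade.
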Material /KatanaGameplay/VFX/M_Katana_Dash_Head
graph {
 properties {
  bUsedWithNiagaraMeshParticles=true
  blendMode=BLEND_Translucent
  pixelDepthOffsetMode=PDOM_Legacy
  shadingModel=MSM_Unlit
  twoSided=true
  usedShadingModels=Unlit
 }
 nodes [
  n1 = ParticleColor()
  n2 = Power(Base=ComponentMask(None=TextureCoordinate()) {g=true}, Exp=ScalarParameter() {defaultValue=1, parameterName=Alpha Contrast})
  n3 = TextureCoordinate() {uTiling=4, vTiling=0.7}
  n4 = VectorParameter() {parameterName=Tile, defaultValue=<rgba 1 1 0 0>}
  n5 = AppendVector(A=@n4.R, B=@n4.R)
  n6 = Multiply(A=Time(), B=ScalarParameter() {defaultValue=0.6, parameterName=PanSpeed})
  n7 = TextureSample(UVs=Panner(Coordinate=Multiply(A=@n3, B=@n5), Time=@n6) {speedX=4, speedY=7}) {samplerType=SAMPLERTYPE_LinearGrayscale, texture=/Game/Global/Textures/Default/Noise/T_SwirlNoise_N_height}
  n8 = TextureCoordinate() {uTiling=3, vTiling=0.6}
  n9 = TextureSample(UVs=Panner(Coordinate=Multiply(A=@n5, B=@n8), Time=@n6) {speedX=-5, speedY=16}) {texture=&T_Smooth_tile}
  n10 = ComponentMask(None=TextureCoordinate()) {g=true}
  n11 = Power(
   Base=Multiply(A=LinearInterpolate(A=@n7.R, B=@n9.R)) {constB=1.5}
   Exp=LinearInterpolate(
    A=ScalarParameter() {defaultValue=1, parameterName=Streaks_LerpA}
    B=ScalarParameter() {defaultValue=3, parameterName=Streaks_LerpB}
    Alpha=OneMinus(None=@n10)
   ) {constA=0.6, constB=4}
  )
  n12 = LinearInterpolate(
   A=VectorParameter() {parameterName=DarkColor, defaultValue=<rgba 0.099 0.4 0.2899 0>}.RGB
   B=VectorParameter() {parameterName=LightColor, defaultValue=<rgba 0 1.5 0.4258 0>}.RGB
   Alpha=Multiply(A=@n2, B=@n11)
  )
  n13 = TextureCoordinate()
  n14 = Multiply(
   A=Power(Base=ComponentMask(None=@n13) {g=true}, Exp=ScalarParameter() {defaultValue=6, parameterName=Central Contast}) {constExponent=10}
   B=VectorParameter() {parameterName=CenterColor, defaultValue=<rgba 0.4531 0.7178 1 0>}.RGB
  )
  n15 = Multiply(A=@n5, B=TextureCoordinate() {uTiling=3})
  n16 = Multiply(A=@n5, B=TextureCoordinate() {uTiling=4, vTiling=2})
  n17 = Add(
   A=TextureSample(UVs=Panner(Coordinate=@n15, Time=@n6) {speedX=-1, speedY=2.5}) {texture=/Game/Global/Textures/Default/PointSplatter/T_Tiling_Dust}.RGB
   B=TextureSample(UVs=Panner(Coordinate=@n16, Time=@n6) {speedX=1, speedY=5}) {texture=&T_Tiling_Dust}.RGB
  )
  n18 = Add(A=Add(A=@n14, B=Clamp(None=@n12) {maxDefault=5}), B=Multiply(A=@n12, B=Multiply(A=@n17, B=@n17)))
  n19 = Clamp(None=@n18)
  n20 = Multiply(
   A=@n18
   B=MaterialFunctionCall(
    Color_Morning=ScalarParameter() {defaultValue=0.7, parameterName=HDRMult_3}
    Color_Day=ScalarParameter() {defaultValue=1, parameterName=HDRMult_2}
    Color_Evening=ScalarParameter() {defaultValue=0.7, parameterName=HDRMult_1}
    Color_Night=ScalarParameter() {defaultValue=0.5, parameterName=HDRMult}
   ) {materialFunction=&Color_Time_of_Day}.Result
  )
  MaterialFunctionCall(Emissive=Multiply(
    A=@n1.RGB
    B=MaterialFunctionCall(LDR Input=@n19, HDR Tint=@n20) {materialFunction=&DeriveHDRfromLDR}.RemappedRange
   )) {materialFunction=&MF_FX_EmissiveAdjustments}
  n21 = TextureCoordinate()
  n22 = Constant() {r=0.5}
  n23 = Add(A=Multiply(A=@n2, B=ScalarParameter() {defaultValue=0.5, parameterName=Alpha Intensity}), B=Add(A=@n14, B=@n11))
  n24 = OneMinus(None=Fresnel(ExponentIn=ScalarParameter() {defaultValue=0.1, parameterName=Fresnel Exp_1}) {baseReflectFraction=0})
  n25 = Multiply(A=ScalarParameter() {defaultValue=0.5, parameterName=Central_Alpha Intensity}, B=@n14)
  n26 = Multiply(
   A=LinearInterpolate(B=SphereMask(A=@n21, B=@n22) {attenuationRadius=0.5, hardnessPercent=50}, Alpha=ScalarParameter() {parameterName=UVEdgeFade}) {constA=1}
   B=Multiply(A=Multiply(A=@n1.A, B=@n23), B=Multiply(A=@n24, B=@n25))
  )
  n27 = ScalarParameter() {defaultValue=2.5, parameterName=Opacity}
  Multiply(
   A=MaterialFunctionCall(Distance=ScalarParameter() {defaultValue=250, parameterName=Near Camera Fade}) {materialFunction=&Near_Camera_Fade}.Result
   B=DepthFade(Opacity=Multiply(A=@n26, B=@n27), FadeDistance=ScalarParameter() {defaultValue=100, parameterName=Depth Fade})
  )
 ]
}
