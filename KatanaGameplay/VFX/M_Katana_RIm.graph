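# M_Katana_RIm is an unlit additive material for Niagara mesh particles, producing an animated, noisy rim glow for katana VFX.
# The rim multiplies inverted normal-to-camera dot product by a powered copy of that dot product, fading the extreme silhouette, then sharpens the result.
# Local-space position, panned over time and scaled, samples a noise texture with three-component coordinates.
# The powered noise modulates the rim, which is brightened and tinted by particle color.
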Material /KatanaGameplay/VFX/M_Katana_RIm
graph {
 properties {
  bUsedWithNiagaraMeshParticles=true
  blendMode=BLEND_Additive
  pixelDepthOffsetMode=PDOM_Legacy
  shadingModel=MSM_Unlit
  usedShadingModels=Unlit
 }
 nodes [
  n1 = DotProduct(A=VertexNormalWS(), B=CameraVectorWS())
  n2 = ScalarParameter() {defaultValue=2, parameterName=Pow_out}
  n3 = Power(Base=Multiply(A=OneMinus(None=@n1), B=Power(Base=@n1, Exp=@n2) {constExponent=1}), Exp=ScalarParameter() {defaultValue=2, parameterName=Pow})
  n4 = Transform(None=WorldPosition().XYZ) {transformSourceType=TRANSFORMSOURCE_World, transformType=TRANSFORM_Local}
  n5 = Multiply(A=VectorParameter() {parameterName=Pan, defaultValue=<rgba 2 1 20 0>}.RGB, B=Time())
  n6 = TextureSample(UVs=Multiply(A=Add(A=@n4, B=@n5), B=ScalarParameter() {defaultValue=0.01, parameterName=3DNoise Scale}) {constB=0.01}) {samplerType=SAMPLERTYPE_LinearColor, texture=/Game/Gadgets/Assets/VinderTech_GliderChute/Glider_Mitosis_Male/Textures/Noise/VT_GaussianNoise}
  n7 = Power(Base=@n6.R)
  Multiply(A=ParticleColor().RGB, B=Multiply(A=Multiply(A=@n3, B=@n7)) {constB=5})
 ]
}
