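# Translucent two-sided katana trail material for Niagara ribbons, sprites and meshes, with a colored inside and dark edge lines.
# A centered V gradient builds soft line masks, with boldness and line transparency driven by dynamic parameters.
# A rotated inside texture, offset by time plus a random value, is tinted by Color and particle color and cut by that mask.
# UV1 fades opacity along the trail, followed by depth fade.
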Material /KatanaGameplay/VFX/M_Katana_Trail_2
graph {
 properties {
  bUsedWithNiagaraMeshParticles=true
  bUsedWithNiagaraRibbons=true
  bUsedWithNiagaraSprites=true
  blendMode=BLEND_Translucent
  pixelDepthOffsetMode=PDOM_Legacy
  twoSided=true
  usedShadingModels=Default Lit
 }
 nodes [
  n1 = ParticleColor()
  n2 = Power(
   Base=ComponentMask(None=TextureCoordinate() {coordinateIndex=1}) {r=true}
   Exp=ScalarParameter() {defaultValue=0.25, parameterName=Head Fade Power}
  )
  n3 = Subtract(A=ComponentMask(None=TextureCoordinate()) {g=true}) {constB=0.5}
  n4 = OneMinus(None=Multiply(A=Abs(None=@n3)) {constB=2})
  n5 = ScalarParameter() {defaultValue=0.1, parameterName=Boldness}
  n6 = DynamicParameter() {defaultValue=<rgba 1 0.9 1 1>, paramNames [Line Bold Mul, Random, Random_Bold_Transparency, Param4]}
  n7 = Saturate(None=Subtract(A=@n4, B=Multiply(A=@n5, B=@n6.Line Bold Mul)) {constB=0.1})
  n8 = ScalarParameter() {defaultValue=12.5, parameterName=Edge Softness}
  n9 = Saturate(None=Multiply(A=@n7, B=@n8) {constB=20})
  n10 = Multiply(A=@n4, B=@n8) {constB=20}
  n11 = Power(Base=Multiply(A=OneMinus(None=@n9), B=Saturate(None=@n10)), Exp=ScalarParameter() {defaultValue=1.25, parameterName=Black Line Pow})
  n12 = ComponentMask(None=TextureCoordinate() {coordinateIndex=1}) {r=true}
  n13 = OneMinus(None=@n12)
  n14 = Add(A=Multiply(A=Multiply(A=@n11, B=@n6.Random_Bold_Transparency), B=@n13), B=Multiply(A=@n13, B=Power(Base=@n12) {constExponent=0.5}))
  DepthFade(Opacity=Multiply(A=@n1.A, B=Multiply(A=@n2, B=@n14)), FadeDistance=ScalarParameter() {defaultValue=100, parameterName=DepthFade})
  n15 = DynamicParameter() {defaultValue=<rgba 1 0.9 1 1>, paramNames [Line Bold Mul, Random, Random_Bold_Transparency, Param4]}
  n16 = Time()
  n17 = MaterialFunctionCall(Float2=TextureCoordinate() {vTiling=0.2}) {materialFunction=&BreakOutFloat2Components}
  n18 = Add(
   A=Multiply(A=Add(A=@n15.Random, B=@n16), B=Constant2Vector() {r=0.05})
   B=Rotator(Coordinate=AppendVector(A=@n17.G, B=@n17.G), Time=ScalarParameter() {defaultValue=2, parameterName=rot})
  )
  n19 = VectorParameter() {parameterName=Color, defaultValue=<rgba 0.1881 0 2 0>}
  n20 = Sign(None=@n7)
  Multiply(
   A=Multiply(
    A=TextureSample(UVs=@n18) {samplerType=SAMPLERTYPE_LinearColor, texture=&/T_Katana_RIbbon_Inside}.RGB
    B=Multiply(A=@n19.RGB, B=@n20)
   )
   B=@n1.RGB
  )
 ]
}
