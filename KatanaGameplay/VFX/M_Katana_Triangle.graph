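# Translucent two-sided Katana VFX material for Niagara sprites, ribbons and mesh particles, drawing a noisy colored band.
# Directional noise scrolls over time with a per-particle random offset, using UVs rotated by a parameter.
# A triangle gradient across U, thresholded by Boldness through Saturate and Sign, forms a hard-edged centered mask.
# The noise, mask, Color parameter and particle color multiply together, while particle alpha fades linearly along V.
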Material /KatanaGameplay/VFX/M_Katana_Triangle
graph {
 properties {
  bUsedWithNiagaraMeshParticles=true
  bUsedWithNiagaraRibbons=true
  bUsedWithNiagaraSprites=true
  blendMode=BLEND_Translucent
  pixelDepthOffsetMode=PDOM_Legacy
  twoSided=true
  usedShadingModels=Default Lit
 }
 nodes [
  n1 = ParticleRandom()
  n2 = Time()
  n3 = Add(
   A=Multiply(A=Add(A=@n1, B=@n2), B=Constant2Vector() {g=0.25, r=0.2})
   B=Rotator(Coordinate=TextureCoordinate() {uTiling=0.2}, Time=ScalarParameter() {defaultValue=5, parameterName=rot})
  )
  n4 = VectorParameter() {parameterName=Color, defaultValue=<rgba 0.1881 0 2 0>}
  n5 = TextureCoordinate()
  n6 = Subtract(A=ComponentMask(None=@n5) {r=true}) {constB=0.5}
  n7 = OneMinus(None=Multiply(A=Abs(None=@n6)) {constB=2})
  n8 = ScalarParameter() {defaultValue=0.2, parameterName=Boldness}
  n9 = Sign(None=Saturate(None=Subtract(A=@n7, B=@n8) {constB=0.1}))
  n10 = ParticleColor()
  Multiply(
   A=Multiply(
    A=TextureSample(UVs=@n3) {samplerType=SAMPLERTYPE_LinearGrayscale, texture=/Game/Effects/Fort_Effects/Emotes/UltraLight/Texture/T_DirectionalNoise_Greyscale}.RGB
    B=Multiply(A=@n4.RGB, B=@n9)
   )
   B=@n10.RGB
  )
  Multiply(A=@n10.A, B=OneMinus(None=ComponentMask(None=@n5) {g=true}))
 ]
}
