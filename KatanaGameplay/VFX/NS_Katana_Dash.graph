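# Niagara system for the Katana dash: short-lived lens-effect meshes, character-following ribbon trails and speed streaks.
# Four ribbon emitters spawn per distance moved on skeletal mesh bones or sockets indexed by spawn group, with widths animated by curves and alpha fading on emitter age.
# GPU sprite streaks use negative point attraction and stretch with speed.
# User.ColorVariant switches most emitters between two palettes; User.SpawnSpacing_ribbon sets trail density.
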NiagaraSystem /KatanaGameplay/VFX/NS_Katana_Dash
dependencies {
 usedDataInterfaces [
  &NiagaraDataInterfaceArrayColor
  &NiagaraDataInterfaceCurve
  &NiagaraDataInterfaceVectorCurve
  &NiagaraDataInterfaceArrayFloat
  &NiagaraDataInterfaceSkeletalMesh
 ]
 usedDynamicInputs []
 usedModules [
  &self:SystemSpawnScript.SystemScriptSource.SystemScriptGraph.NiagaraNodeAssignment_0.SetVariables_#1
  &SystemState
 ]
}
emitters {
 LensEffect {
  emitterUpdateScript [
   &EmitterState(
    Inactive Response=::NewEnumerator1
    Life Cycle Mode=::NewEnumerator1
    Loop Behavior=::NewEnumerator1
    Loop Delay=0.04
    Loop Duration=0.1
    UseLoopDelay=true
   )
   &SpawnRate(SpawnRate=60, Use Spawn Probability=false)
   &SpawnBurst_Instantaneous(Spawn Count=1, Use Spawn Probability=false)
  ]
  particleSpawnScript [
   /Niagara/Modules/Spawn/Initialization/InitializeParticle(
    Color=&SelectLinearColorFromArray(
     Array Sampling Mode=::NewEnumerator1
     Direct Array Index=$User.ColorVariant
     Color Selection Array=DataInterface() {propertyValues {ColorData [<rgba 0.0799 0 1.5 1>, <rgba 0.3854 0 0.0065 1>]}}
     Random Seed=?
    )
    Lifetime=0.4
    Mesh Scale=<1.25 1.25 2>
    Module.Write Color=true
    Module.Write Lifetime=true
    Module.Write Mass=true
    Module.Write Material Random=true
    Module.Write Position=false
    Module.Write Scale=true
    Module.Write SpriteRotation=false
    Module.Write SpriteSize=false
    Position=&SimulationPosition()
   )
   &InitialMeshOrientation(Mesh Orientation Mode=::NewEnumerator2, Orientation Coordinate Space=::Local, Use Rotation Vector=false)
  ]
  particleUpdateScript [
   &ParticleState()
   &ScaleMeshSize(Scale Factor=&Multiply_VectorByFloat(
     Vector=<1 1 1>
     Float=&FloatFromCurve(FloatCurve=DataInterface() {propertyValues {Curve {keys [{tangentMode=RCTM_User, time=0, value=0.5}, {tangentMode=RCTM_User, time=1, value=2}]}}})
    ))
   &ScaleColor(
    Scale Alpha=&FloatFromCurve(FloatCurve=DataInterface() {propertyValues {Curve {keys [{arriveTangent=6.3626, interpMode=RCIM_Cubic, time=0, value=1}, {interpMode=RCIM_Cubic, leaveTangent=-0.6839, time=1, value=0}]}}})
    Scale RGB=&VectorFromCurve(VectorCurve=DataInterface() {propertyValues {XCurve {keys [<at 0 1 RCIM_Cubic>,<at 1 0>]}, YCurve {keys [<at 0 1 RCIM_Cubic>,<at 1 0>]}, ZCurve {keys [<at 0 1 RCIM_Cubic>,<at 1 0>]}}})
    ScaleA=true
    ScaleRGB=true
   )
  ]
  rendererClasses [&NiagaraMeshRendererProperties]
  renderers [{rendererClass=&NiagaraMeshRendererProperties, rendererIndex=0}]
 }
 LensEffect001 {
  emitterUpdateScript [
   &EmitterState(
    Inactive Response=::NewEnumerator1
    Life Cycle Mode=::NewEnumerator1
    Loop Behavior=::NewEnumerator1
    Loop Delay=0.05
    Loop Duration=0.8
    UseLoopDelay=true
   )
   &SpawnBurst_Instantaneous(Spawn Count=1, Use Spawn Probability=false)
  ]
  particleSpawnScript [
   &InitializeParticle(
    Color=<rgba 0 0 0 1>
    Lifetime=0.35
    Mesh Scale=<4 1 1.5>
    Module.Write Color=true
    Module.Write Lifetime=true
    Module.Write Mass=true
    Module.Write Material Random=true
    Module.Write Position=false
    Module.Write Scale=true
    Module.Write SpriteRotation=false
    Module.Write SpriteSize=false
    Position=&SimulationPosition()
   )
   &InitialMeshOrientation(Mesh Orientation Mode=::NewEnumerator7, Rotation=<0 0 -0.25>, Use Rotation Vector=true) {module=InitialMeshOrientation001}
  ]
  particleUpdateScript [
   &ParticleState()
   &SkeletalMeshLocation(
    Bone / Socket Index=7
    "Inherit Velocity (Scale)"=0.2
    MaskSpawnGroup=false
    Mesh Sampling Type=::NewEnumerator5
    Orientation Sampling=::NewEnumerator1
    Sampled Position Offset=<100 0 120>
    Skeletal Mesh=$System.Skeletal Mesh
    Socket Sampling Mode=::NewEnumerator5
    Velocity Sampling=::NewEnumerator0
    VelocitySpeedLimit=false
   )
   &ScaleColor(
    Scale Alpha=&FloatFromCurve(
     FloatCurve=DataInterface() {propertyValues {Curve {keys [{arriveTangent=6.3626, interpMode=RCIM_Cubic, time=0, value=1}, {interpMode=RCIM_Cubic, leaveTangent=-0.6839, time=1, value=0}]}}}
     Scale Curve=1.04
    )
    Scale RGB=&VectorFromCurve(VectorCurve=DataInterface() {propertyValues {XCurve {keys [<at 0 1 RCIM_Cubic>,<at 1 0>]}, YCurve {keys [<at 0 1 RCIM_Cubic>,<at 1 0>]}, ZCurve {keys [<at 0 1 RCIM_Cubic>,<at 1 0>]}}})
    ScaleA=true
    ScaleRGB=true
   )
  ]
  rendererClasses [&NiagaraMeshRendererProperties]
  renderers [{rendererClass=&NiagaraMeshRendererProperties, rendererIndex=0}]
 }
 Ribbon_Blue {
  emitterUpdateScript [
   &EmitterState(Life Cycle Mode=::NewEnumerator1, Loop Behavior=::NewEnumerator1, Loop Delay=0.05, Loop Duration=0.55, UseLoopDelay=true)
   &SpawnPerUnit(
    Max Movement Threshold=50
    Module.Use Max Movement Threshold=false
    Module.Use Movement Tolerance=true
    Movement Tolerance=40
    Spawn Probability=1
    Spawn Spacing:NiagaraFloat=$User.SpawnSpacing_ribbon
    Use Spawn Probability=true
   )
   &SpawnPerUnit(
    Max Movement Threshold=50
    Module.Use Max Movement Threshold=false
    Module.Use Movement Tolerance=true
    Movement Tolerance=40
    Spawn Group=3
    Spawn Probability=1
    Spawn Spacing:NiagaraFloat=$User.SpawnSpacing_ribbon
    Use Spawn Probability=true
   ) {module=SpawnPerUnit003}
   &SpawnPerUnit(
    Max Movement Threshold=50
    Module.Use Max Movement Threshold=false
    Module.Use Movement Tolerance=true
    Movement Tolerance=40
    Spawn Group=4
    Spawn Probability=1
    Spawn Spacing:NiagaraFloat=$User.SpawnSpacing_ribbon
    Use Spawn Probability=true
   ) {module=SpawnPerUnit004}
   &SpawnPerUnit(
    Max Movement Threshold=50
    Module.Use Max Movement Threshold=false
    Module.Use Movement Tolerance=true
    Movement Tolerance=40
    Spawn Group=2
    Spawn Probability=1
    Spawn Spacing:NiagaraFloat=$User.SpawnSpacing_ribbon
    Use Spawn Probability=true
   ) {module=SpawnPerUnit002}
   &SpawnPerUnit(
    Max Movement Threshold=50
    Module.Use Max Movement Threshold=false
    Module.Use Movement Tolerance=true
    Movement Tolerance=40
    Spawn Group=1
    Spawn Probability=1
    Spawn Spacing:NiagaraFloat=$User.SpawnSpacing_ribbon
    Use Spawn Probability=true
   ) {module=SpawnPerUnit001}
   /SetVariables(Emitter.Color:LinearColor=&SelectLinearColorFromArray(
     Array Sampling Mode=::NewEnumerator1
     Direct Array Index=$User.ColorVariant
     Color Selection Array=DataInterface() {propertyValues {ColorData [<rgba 0.1881 0 2 1>, <rgba 0.5 0.0211 0 1>]}}
     Random Seed=?
    )) {isSetParameters=true}
  ]
  particleSpawnScript [
   /Niagara/Modules/Spawn/Initialization/V2/InitializeParticle(
    Color=<rgba 2 2 2 1>
    Ribbon Width=&SelectFloatFromArray(Direct Array Index=$Emitter.SpawnGroup, Float Selection Array=DataInterface() {propertyValues {FloatData [20,35,60,30,50]}})
    Ribbon Width Mode=::NewEnumerator1
    Sprite Size Mode=::NewEnumerator3
    Uniform Sprite Size=50
    UsePositionOffset=false
    Write Material Random=false
   )
   &AddVelocity(Velocity=<200 0 0>)
   &SkeletalMeshLocation(
    Bone / Socket Index=&Add_Integer(A=$Emitter.SpawnGroup, B=0)
    Bone Sampling Mode=::NewEnumerator4
    MaskSpawnGroup=false
    Skeletal Mesh=$System.Skeletal Mesh
   )
  ]
  particleUpdateScript [
   &ParticleState()
   &ScaleColor(
    Scale Alpha=&FloatFromCurve(
     FloatCurve=DataInterface() {propertyValues {Curve {keys [
         <at 0 0 RCIM_Cubic>
         <at 0.25 1>
         {arriveTangent=-1.8146, interpMode=RCIM_Cubic, leaveTangent=-1.8146, time=0.5, value=1.0113}
         <at 0.8 0.002 RCIM_Cubic>
        ]}}}
     CurveIndex=$Emitter.Age
     Scale Curve=0.9
    )
    ScaleA=true
    ScaleRGB=false
   )
   &SolveForcesAndVelocity(Clamp Velocity=false, Limit Acceleration=false)
   &ScaleRibbonWidth(Ribbon Width Scale=&FloatFromCurve(
     FloatCurve=DataInterface() {propertyValues {Curve {keys [
         {arriveTangent=2.3626, interpMode=RCIM_Cubic, leaveTangent=2.3626, tangentMode=RCTM_User, time=0, value=0.74}
         <at 0.5 5>
         {arriveTangent=15.9562, interpMode=RCIM_Cubic, leaveTangent=15.9562, tangentMode=RCTM_User, time=1, value=0}
        ]}}}
     Scale Curve=&FloatFromCurve(FloatCurve=DataInterface() {propertyValues {Curve {keys [<at 0 1>,<at 0.8 0>]}}}, CurveIndex=$Emitter.Age)
    ))
   &DynamicMaterialParameters(
    Index 0 Param 1=&RandomRangeFloat(Minimum=0.2, Maximum=3.2)
    Index 0 Param 2=&RandomRangeFloat(Minimum=0, Maximum=1)
    Index 0 Param 3=&RandomRangeFloat(Minimum=0.2, Maximum=1.05)
    Module.Param0WriteEnabled=true
    Module.Param1WriteEnabled=true
    Module.Param2WriteEnabled=true
    Module.Param3WriteEnabled=false
   )
  ]
  rendererClasses [&NiagaraRibbonRendererProperties]
  renderers [{rendererClass=&NiagaraRibbonRendererProperties}]
 }
 Ribbon_Blue001 {
  emitterUpdateScript [
   &EmitterState(Life Cycle Mode=::NewEnumerator1, Loop Behavior=::NewEnumerator1, Loop Delay=0.05, Loop Duration=0.55, UseLoopDelay=true)
   &SpawnPerUnit(
    Max Movement Threshold=50
    Module.Use Max Movement Threshold=false
    Module.Use Movement Tolerance=true
    Movement Tolerance=50
    Spawn Probability=1
    Spawn Spacing:NiagaraFloat=$User.SpawnSpacing_ribbon
    Use Spawn Probability=true
   )
   &SpawnPerUnit(
    Max Movement Threshold=50
    Module.Use Max Movement Threshold=false
    Module.Use Movement Tolerance=true
    Movement Tolerance=50
    Spawn Group=3
    Spawn Probability=1
    Spawn Spacing:NiagaraFloat=$User.SpawnSpacing_ribbon
    Use Spawn Probability=true
   ) {module=SpawnPerUnit003}
   &SpawnPerUnit(
    Max Movement Threshold=50
    Module.Use Max Movement Threshold=false
    Module.Use Movement Tolerance=true
    Movement Tolerance=50
    Spawn Group=2
    Spawn Probability=1
    Spawn Spacing:NiagaraFloat=$User.SpawnSpacing_ribbon
    Use Spawn Probability=true
   ) {module=SpawnPerUnit002}
   /SetVariables(Emitter.Color:LinearColor=&SelectLinearColorFromArray(
     Array Sampling Mode=::NewEnumerator1
     Direct Array Index=$User.ColorVariant
     Color Selection Array=DataInterface() {propertyValues {ColorData [<rgba 0.1881 0 2 1>, <rgba 0.5 0.0211 0 1>]}}
     Random Seed=?
    )) {isSetParameters=true}
  ]
  particleSpawnScript [
   &InitializeParticle(
    Color=<rgba 2 2 2 1>
    Ribbon Width=&SelectFloatFromArray(Direct Array Index=$Emitter.SpawnGroup, Float Selection Array=DataInterface() {propertyValues {FloatData [20,35,60,30,50]}})
    Ribbon Width Mode=::NewEnumerator1
    Sprite Size Mode=::NewEnumerator3
    Uniform Sprite Size=50
    UsePositionOffset=false
    Write Material Random=false
   )
   &SkeletalMeshLocation(
    Bone / Socket Index=&Add_Integer(A=$Emitter.SpawnGroup, B=0)
    Bone Sampling Mode=::NewEnumerator4
    MaskSpawnGroup=false
    Skeletal Mesh=$System.Skeletal Mesh
   )
   &AddVelocity(Velocity=<200 0 0>)
  ]
  particleUpdateScript [
   &ParticleState()
   &ScaleColor(
    Scale Alpha=&FloatFromCurve(
     FloatCurve=DataInterface() {propertyValues {Curve {keys [
         <at 0 0 RCIM_Cubic>
         <at 0.25 1>
         {arriveTangent=-1.8146, interpMode=RCIM_Cubic, leaveTangent=-1.8146, time=0.5, value=1.0113}
         <at 0.8 0.002 RCIM_Cubic>
        ]}}}
     CurveIndex=$Emitter.Age
     Scale Curve=0.9
    )
    ScaleA=true
    ScaleRGB=false
   )
   &SolveForcesAndVelocity(Clamp Velocity=false, Limit Acceleration=false)
   &ScaleRibbonWidth(Ribbon Width Scale=&FloatFromCurve(
     FloatCurve=DataInterface() {propertyValues {Curve {keys [
         {arriveTangent=2.3626, interpMode=RCIM_Cubic, leaveTangent=2.3626, tangentMode=RCTM_User, time=0, value=0.74}
         <at 0.5 5>
         {arriveTangent=15.9562, interpMode=RCIM_Cubic, leaveTangent=15.9562, tangentMode=RCTM_User, time=1, value=0}
        ]}}}
     Scale Curve=&FloatFromCurve(FloatCurve=DataInterface() {propertyValues {Curve {keys [<at 0 1>,<at 0.8 0>]}}}, CurveIndex=$Emitter.Age)
    ))
   &DynamicMaterialParameters(
    Index 0 Param 1=&RandomRangeFloat(Minimum=0.2, Maximum=3.2)
    Index 0 Param 2=&RandomRangeFloat(Minimum=0, Maximum=1)
    Index 0 Param 3=&RandomRangeFloat(Minimum=0.2, Maximum=1.05)
    Module.Param0WriteEnabled=true
    Module.Param1WriteEnabled=true
    Module.Param2WriteEnabled=true
    Module.Param3WriteEnabled=false
   )
  ]
  rendererClasses [&NiagaraRibbonRendererProperties]
  renderers [{rendererClass=&NiagaraRibbonRendererProperties}]
 }
 Ribbon_Pink {
  emitterUpdateScript [
   &EmitterState(Life Cycle Mode=::NewEnumerator1, Loop Behavior=::NewEnumerator1, Loop Delay=0.05, Loop Duration=0.1583, UseLoopDelay=true)
   &SpawnPerUnit(
    Max Movement Threshold=50
    Module.Use Max Movement Threshold=false
    Module.Use Movement Tolerance=true
    Movement Tolerance=40
    Spawn Probability=1
    Spawn Spacing:NiagaraFloat=$User.SpawnSpacing_ribbon
    Use Spawn Probability=true
   )
   &SpawnPerUnit(
    Max Movement Threshold=50
    Module.Use Max Movement Threshold=false
    Module.Use Movement Tolerance=true
    Movement Tolerance=40
    Spawn Group=1
    Spawn Probability=1
    Spawn Spacing:NiagaraFloat=$User.SpawnSpacing_ribbon
    Use Spawn Probability=true
   ) {module=SpawnPerUnit001}
   &SpawnPerUnit(
    Max Movement Threshold=50
    Module.Use Max Movement Threshold=false
    Module.Use Movement Tolerance=true
    Movement Tolerance=40
    Spawn Group=2
    Spawn Probability=1
    Spawn Spacing:NiagaraFloat=$User.SpawnSpacing_ribbon
    Use Spawn Probability=true
   ) {module=SpawnPerUnit002}
   &SpawnPerUnit(
    Max Movement Threshold=50
    Module.Use Max Movement Threshold=false
    Module.Use Movement Tolerance=true
    Movement Tolerance=40
    Spawn Group=4
    Spawn Probability=1
    Spawn Spacing:NiagaraFloat=$User.SpawnSpacing_ribbon
    Use Spawn Probability=true
   ) {module=SpawnPerUnit004}
   &SpawnPerUnit(
    Max Movement Threshold=50
    Module.Use Max Movement Threshold=false
    Module.Use Movement Tolerance=true
    Movement Tolerance=40
    Spawn Group=3
    Spawn Probability=1
    Spawn Spacing:NiagaraFloat=$User.SpawnSpacing_ribbon
    Use Spawn Probability=true
   ) {module=SpawnPerUnit003}
  ]
  particleSpawnScript [
   &InitializeParticle(
    Color=&SelectLinearColorFromArray(
     Array Sampling Mode=::NewEnumerator1
     Direct Array Index=$User.ColorVariant
     Color Selection Array=DataInterface() {propertyValues {ColorData [<rgba 1 1 1 1>, <rgba 0.6198 0.0251 0 1>]}}
     Random Seed=?
    )
    Lifetime=0.5
    Ribbon Width Mode=::NewEnumerator1
    Sprite Size Mode=::NewEnumerator3
    Uniform Sprite Size=50
    UsePositionOffset=false
    Write Material Random=false
   )
   &SkeletalMeshLocation(
    Bone / Socket Index=&Add_Integer(A=$Emitter.SpawnGroup, B=0)
    MaskSpawnGroup=false
    Mesh Sampling Type=::NewEnumerator5
    Skeletal Mesh=$System.Skeletal Mesh
    Socket Sampling Mode=::NewEnumerator5
   )
  ]
  particleUpdateScript [
   &ParticleState()
   &ScaleColor(
    Scale Alpha=&FloatFromCurve(FloatCurve=DataInterface() {propertyValues {Curve {keys [<at 0.3 1.0113 RCIM_Cubic>,<at 0.5 0.002 RCIM_Cubic>]}}}, CurveIndex=$Emitter.Age)
    ScaleA=true
    ScaleRGB=false
   )
   &SolveForcesAndVelocity(Clamp Velocity=false, Limit Acceleration=false)
   &ScaleRibbonWidth(Ribbon Width Scale=&FloatFromCurve(
     FloatCurve=DataInterface() {propertyValues {Curve {keys [{interpMode=RCIM_Cubic, tangentMode=RCTM_User, time=0, value=0}, {interpMode=RCIM_Cubic, tangentMode=RCTM_User, time=1, value=10}]}}}
     Scale Curve=2
    ))
   &DynamicMaterialParameters(
    Index 0 Param 3=&RandomRangeFloat(Minimum=0, Maximum=1)
    Module.Param0WriteEnabled=false
    Module.Param1WriteEnabled=false
    Module.Param2WriteEnabled=true
    Module.Param3WriteEnabled=true
   )
  ]
  rendererClasses [&NiagaraRibbonRendererProperties]
  renderers [{rendererClass=&NiagaraRibbonRendererProperties}]
 }
 Ribbon_Pink001 {
  emitterUpdateScript [
   &EmitterState(Life Cycle Mode=::NewEnumerator1, Loop Behavior=::NewEnumerator1, Loop Delay=0.21, Loop Duration=0.45, UseLoopDelay=true)
   &SpawnPerUnit(
    Max Movement Threshold=50
    Module.Use Max Movement Threshold=false
    Module.Use Movement Tolerance=true
    Movement Tolerance=50
    Spawn Probability=1
    Spawn Spacing:NiagaraFloat=$User.SpawnSpacing_ribbon
    Use Spawn Probability=true
   )
   &SpawnPerUnit(
    Max Movement Threshold=50
    Module.Use Max Movement Threshold=false
    Module.Use Movement Tolerance=true
    Movement Tolerance=50
    Spawn Group=2
    Spawn Probability=1
    Spawn Spacing:NiagaraFloat=$User.SpawnSpacing_ribbon
    Use Spawn Probability=true
   ) {module=SpawnPerUnit002}
   &SpawnPerUnit(
    Max Movement Threshold=50
    Module.Use Max Movement Threshold=false
    Module.Use Movement Tolerance=true
    Movement Tolerance=50
    Spawn Group=3
    Spawn Probability=1
    Spawn Spacing:NiagaraFloat=$User.SpawnSpacing_ribbon
    Use Spawn Probability=true
   ) {module=SpawnPerUnit003}
  ]
  particleSpawnScript [
   &InitializeParticle(
    Color=&SelectLinearColorFromArray(
     Array Sampling Mode=::NewEnumerator1
     Direct Array Index=$User.ColorVariant
     Color Selection Array=DataInterface() {propertyValues {ColorData [<rgba 1 1 1 1>, <rgba 0.6198 0.0251 0 1>]}}
     Random Seed=?
    )
    Lifetime=0.5
    Ribbon Width Mode=::NewEnumerator1
    Sprite Size Mode=::NewEnumerator3
    Uniform Sprite Size=50
    UsePositionOffset=false
    Write Material Random=false
   )
   &SkeletalMeshLocation(
    Bone / Socket Index=&Add_Integer(A=$Emitter.SpawnGroup, B=0)
    MaskSpawnGroup=false
    Mesh Sampling Type=::NewEnumerator5
    Skeletal Mesh=$System.Skeletal Mesh
    Socket Sampling Mode=::NewEnumerator5
   )
  ]
  particleUpdateScript [
   &ParticleState()
   &ScaleColor(
    Scale Alpha=&FloatFromCurve(FloatCurve=DataInterface() {propertyValues {Curve {keys [<at 0.3 1.0113 RCIM_Cubic>,<at 0.5 0.002 RCIM_Cubic>]}}}, CurveIndex=$Emitter.Age)
    ScaleA=true
    ScaleRGB=false
   )
   &SolveForcesAndVelocity(Clamp Velocity=false, Limit Acceleration=false)
   &ScaleRibbonWidth(Ribbon Width Scale=&FloatFromCurve(
     FloatCurve=DataInterface() {propertyValues {Curve {keys [{interpMode=RCIM_Cubic, tangentMode=RCTM_User, time=0, value=0}, {interpMode=RCIM_Cubic, tangentMode=RCTM_User, time=1, value=10}]}}}
     Scale Curve=2
    ))
   &DynamicMaterialParameters(
    Index 0 Param 3=&RandomRangeFloat(Minimum=0, Maximum=1)
    Module.Param0WriteEnabled=false
    Module.Param1WriteEnabled=false
    Module.Param2WriteEnabled=true
    Module.Param3WriteEnabled=true
   )
  ]
  rendererClasses [&NiagaraRibbonRendererProperties]
  renderers [{rendererClass=&NiagaraRibbonRendererProperties}]
 }
 Streaks001 {
  simTarget=GPUComputeSim
  emitterUpdateScript [
   &EmitterState(Inactive Response=::NewEnumerator1, Life Cycle Mode=::NewEnumerator1, Loop Behavior=::NewEnumerator1, Loop Duration=0.5, UseLoopDelay=false)
   &SpawnRate(
    SpawnRate=&FloatFromCurve(FloatCurve=DataInterface() {propertyValues {Curve {keys [<at 0 1>,<at 1 0>]}}}, CurveIndex=$Emitter.NormalizedLoopAge, Scale Curve=800)
    Use Spawn Probability=false
   )
   &SpawnRate(
    SpawnGroup=1
    SpawnRate=&FloatFromCurve(FloatCurve=DataInterface() {propertyValues {Curve {keys [<at 0 1>,<at 1 0>]}}}, CurveIndex=$Emitter.NormalizedLoopAge, Scale Curve=800)
    Use Spawn Probability=false
   ) {module=SpawnRate001}
  ]
  particleSpawnScript [
   /Niagara/Modules/Spawn/Initialization/InitializeParticle(
    Color=&MaskLinearColorBySpawnGroup(
     Masked Linear Color=&SelectLinearColorFromArray(
      Array Sampling Mode=::NewEnumerator1
      Direct Array Index=$User.ColorVariant
      Color Selection Array=DataInterface() {propertyValues {ColorData [<rgba 0.5 0.5 0.5 0.5>, <rgba 0.1354 0.1354 0.1354 0.5>]}}
      Random Seed=?
     )
     Passthrough Linear Color=&SelectLinearColorFromArray(
      Array Sampling Mode=::NewEnumerator1
      Direct Array Index=$User.ColorVariant
      Color Selection Array=DataInterface() {propertyValues {ColorData [<rgba 3.0465 0 16 0.5>, <rgba 15 6.6413 0 0.5>]}}
      Random Seed=?
     )
    )
    Lifetime=&UniformRangedFloat(Minimum=0.05, Maximum=0.1)
    Module.Write Color=true
    Module.Write Lifetime=true
    Module.Write Mass=true
    Module.Write Material Random=true
    Module.Write Position=false
    Module.Write Scale=false
    Module.Write SpriteRotation=false
    Module.Write SpriteSize=true
    Position=&SimulationPosition()
    Sprite Size=&MaskVector2dBySpawnGroup(
     Masked Vector2D=&UniformRangedVector2D(Minimum=<0.5 1>, Maximum=<1 1>)
     Passthrough Vector2D=&UniformRangedVector2D(Minimum=<1 1>, Maximum=<1.5 1>)
     Spawn Group Mask=0
    )
   )
   &SystemLocation(
    MaskSpawnGroup=false
    Offset=&Add_Vector(A=<0 -200 0>, B=&UniformRangedVector(Minimum=<-100 -100 -75>, Maximum=<100 100 75>))
   )
  ]
  particleUpdateScript [
   &ParticleState()
   &SkeletalMeshLocation(
    Bone / Socket Index=2
    Bone Sampling Mode=::NewEnumerator4
    "Inherit Velocity (Scale)"=-1
    MaskSpawnGroup=false
    Orientation Sampling=::NewEnumerator1
    Sampled Position Offset=&RandomRangeVector(Minimum=<-200 -200 50>, Maximum=<200 200 -250>)
    Skeletal Mesh=$System.Skeletal Mesh
    Velocity Sampling=::NewEnumerator0
    VelocitySpeedLimit=false
   )
   &PointAttractionForce(Attraction Radius=5000, Attraction Strength=-800, Kill Within Radius=false, Module.Use Falloff=false, UseAttractorOffset=false)
   &SolveForcesAndVelocity(Clamp Velocity=false, Limit Acceleration=false)
   &ScaleSpriteSizeBySpeed(Max Scale Factor=<1 180>, Min Scale Factor=<0.5 125>)
   &ScaleColor(
    Scale Alpha=&FloatFromCurve(FloatCurve=DataInterface() {propertyValues {Curve {keys [
         {arriveTangent=6.3626, interpMode=RCIM_Cubic, time=0, value=0}
         <at 0.5 0.9779 RCIM_Cubic>
         {interpMode=RCIM_Cubic, leaveTangent=-0.6839, time=1, value=0}
        ]}}})
    ScaleA=true
    ScaleRGB=false
   )
  ]
  rendererClasses [&NiagaraSpriteRendererProperties]
  renderers [{rendererClass=&NiagaraSpriteRendererProperties}]
 }
}
summary {
 systemName=NS_Katana_Dash
 emitters [
  {emitterName=LensEffect, rendererClasses [&NiagaraMeshRendererProperties]}
  {emitterName=LensEffect001, rendererClasses [&NiagaraMeshRendererProperties]}
  {emitterName=Ribbon_Pink, rendererClasses [&NiagaraRibbonRendererProperties]}
  {emitterName=Ribbon_Blue, rendererClasses [&NiagaraRibbonRendererProperties]}
  {emitterName=Streaks001, simTarget=GPUComputeSim, rendererClasses [&NiagaraSpriteRendererProperties]}
  {emitterName=Ribbon_Pink001, rendererClasses [&NiagaraRibbonRendererProperties]}
  {emitterName=Ribbon_Blue001, rendererClasses [&NiagaraRibbonRendererProperties]}
 ]
 userVariables [
  User.bIsFrontEnd {type=NiagaraBool}
  User.bIsFrontEndPreview {type=NiagaraBool}
  User.bisFullyDeployed {type=NiagaraBool}
  User.Color {type=LinearColor, description="", defaultValue {struct=LinearColor, value=<rgba 1 1 1 1>}}
  User.ColorVariant {type=NiagaraInt32}
  User.CustomColorIntensity {type=NiagaraFloat, defaultValue=0, description=""}
  User.ForwardDot {type=NiagaraFloat, defaultValue=0, description=""}
  User.RibbonIDOffset {type=NiagaraInt32}
  User.RightDot {type=NiagaraFloat, defaultValue=0, description=""}
  User.SpawnSpacing_ribbon {type=NiagaraFloat, defaultValue=150, description=""}
  User.TrailOpacity {type=NiagaraFloat, defaultValue=0, description=""}
  User.UseCustomColor {type=NiagaraFloat, defaultValue=0, description=""}
 ]
}
