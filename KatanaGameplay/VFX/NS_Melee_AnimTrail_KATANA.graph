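# Niagara melee swing trail for the katana, drawn as short-lived ribbons that follow sockets or bones on a skeletal mesh.
# One emitter spawns and updates anim-trail particles from a bone pair and feeds three ribbon renderers; a second pair is disabled.
# User.ColorVariant picks between two palettes, including time-of-day tints with optional custom color override.
# Curves fade color and alpha, scaled by User.TrailOpacity, and darken a secondary Color2 over age.
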NiagaraSystem /KatanaGameplay/VFX/NS_Melee_AnimTrail_KATANA
dependencies {
 usedDataInterfaces [&NiagaraDataInterfaceArrayColor, &NiagaraDataInterfaceColorCurve, &NiagaraDataInterfaceCurve]
 usedDynamicInputs []
 usedModules [
  &self:SystemSpawnScript.SystemScriptSource.SystemScriptGraph.NiagaraNodeAssignment_0.SetVariables_#1
  &SystemState
 ]
}
emitters {FN_AnimTrail {
  emitterSpawnScript [&FN_RibbonIDOffsetUtility()]
  emitterUpdateScript [
   &EmitterState()
   &FN_SpawnAnimTrail(Spawn Interval=0)
  ]
  particleSpawnScript [
   &FN_InitializeAnimTrailBoneIndecies(First Bone or Socket Index=1, Second Bone or Socket Index=0)
   &FN_InitializeAnimTrailBoneIndecies(AnimTrail Selection=1, First Bone or Socket Index=2, Second Bone or Socket Index=3) {enabled=false, module=FN_InitializeAnimTrailBoneIndecies001}
   &FN_InitializeAnimTrail(
    Color=&SelectLinearColorFromArray(
     Array Sampling Mode=::NewEnumerator1
     Direct Array Index=$User.ColorVariant
     Color Selection Array=DataInterface() {propertyValues {ColorData [<rgba 1.5654 0 4 0.1>, <rgba 4 0.1414 0 0.1>]}}
     Random Seed=?
    )
    Lifetime=0.2
    SkeletalMesh=$System.Skeletal Mesh
   )
   &AddVelocity(Velocity=<0 400 0>) {enabled=false}
   &self:TrailColor(
    Apply TOD colors to Custom Color=0.5
    Color Day:LinearColor=&SelectLinearColorFromArray(
     Array Sampling Mode=::NewEnumerator1
     Direct Array Index=$User.ColorVariant
     Color Selection Array=DataInterface() {propertyValues {ColorData [<rgba 0.7912 2.0754 3 0.5>, <rgba 2.5492 5 0 0.5>]}}
     Random Seed=?
    )
    Color Evening:LinearColor=&SelectLinearColorFromArray(
     Array Sampling Mode=::NewEnumerator1
     Direct Array Index=$User.ColorVariant
     Color Selection Array=DataInterface() {propertyValues {ColorData [<rgba 2 1.5492 1.3674 0.5>, <rgba 2.0204 3 0 0.5>]}}
     Random Seed=?
    )
    Color Morning:LinearColor=&SelectLinearColorFromArray(
     Array Sampling Mode=::NewEnumerator1
     Direct Array Index=$User.ColorVariant
     Color Selection Array=DataInterface() {propertyValues {ColorData [<rgba 1 0.6836 0.5932 0.5>, <rgba 1.2445 2 0 0.5>]}}
     Random Seed=?
    )
    Color Night:LinearColor=&SelectLinearColorFromArray(
     Array Sampling Mode=::NewEnumerator1
     Direct Array Index=$User.ColorVariant
     Color Selection Array=DataInterface() {propertyValues {ColorData [<rgba 1.6108 1.8993 2 0.5>, <rgba 1.2173 3 0 0.5>]}}
     Random Seed=?
    )
    Custom Color:LinearColor=$User.Color
    Custom Color Intensity:NiagaraFloat=$User.CustomColorIntensity
    FortniteNPC_FortniteActiveTimeOfDay:LinearColor=$NPC.FortniteNPC.FortniteActiveTimeOfDay
    Intensity Multiplier=2
    Use Custom Color:NiagaraFloat=$User.UseCustomColor
   ) {module=TrailColor}
  ]
  particleUpdateScript [
   &SolveForcesAndVelocity(Clamp Velocity=false, Limit Acceleration=false)
   &FN_UpdateAnimTrail(Locked Particle RibbonWidth=1, SkeletalMesh=$System.Skeletal Mesh) {module=FN_UpdateAnimTrail001}
   &ScaleColor(
    Scale Alpha=&FloatFromCurve(
     FloatCurve=DataInterface() {propertyValues {Curve {keys [
         {arriveTangent=-3.0207, interpMode=RCIM_Cubic, leaveTangent=-3.0207, tangentMode=RCTM_User, time=0, value=1}
         {arriveTangent=-1.1093, interpMode=RCIM_Cubic, leaveTangent=-1.1093, tangentMode=RCTM_User, time=0.3, value=0}
        ]}}}
     CurveIndex=$Particles.Age
     Scale Curve=$User.TrailOpacity
    )
    Scale RGB=&MakeVectorFromLinearColorRGB(Color=&ColorFromCurve(ColorCurve=DataInterface() {propertyValues {
        AlphaCurve {keys []}
        BlueCurve {keys [<at 0 2 RCIM_Cubic>,<at 0.8486 0.1449 RCIM_Cubic>]}
        GreenCurve {keys [<at 0 0 RCIM_Cubic>,<at 0.8486 0 RCIM_Cubic>]}
        RedCurve {keys [<at 0 1.1697 RCIM_Cubic>,<at 0.8486 0.1608 RCIM_Cubic>]}
       }}))
    ScaleA=true
    ScaleRGB=true
   )
   /SetVariables(Particles.Color2=&MakeLinearColorFromVectorAndFloat(
     "Vector (RGB)"=&MakeVectorFromLinearColorRGB(Color=&ScaleLinearColorByCurve(
       InitialValue=&SelectLinearColorFromArray(
        Array Sampling Mode=::NewEnumerator1
        Direct Array Index=$User.ColorVariant
        Color Selection Array=DataInterface() {propertyValues {ColorData [<rgba 0.1158 0 0.1979 1>, <rgba 0.1354 0.0043 0 1>]}}
        Random Seed=?
       )
       ScaleCurve=DataInterface() {propertyValues {
         AlphaCurve {keys []}
         BlueCurve {keys [<at 0 1>,<at 0.85 0.02>]}
         GreenCurve {keys [<at 0 1>,<at 0.85 0.02>]}
         RedCurve {keys [<at 0 1>,<at 0.85 0.02>]}
        }}
       CurveIndex=$Particles.NormalizedAge
      ))
     "Float (Alpha)"=10
    )) {isSetParameters=true}
  ]
  rendererClasses [&NiagaraRibbonRendererProperties]
  renderers [
   {rendererClass=&NiagaraRibbonRendererProperties}
   {rendererClass=&NiagaraRibbonRendererProperties, rendererIndex=1}
   {rendererClass=&NiagaraRibbonRendererProperties, rendererIndex=2}
  ]
 }}
summary {
 systemName=NS_Melee_AnimTrail_KATANA
 emitters [{emitterName=FN_AnimTrail, rendererClasses [&NiagaraRibbonRendererProperties]}]
 userVariables [
  User.Color {type=LinearColor, description="", defaultValue {struct=LinearColor, value=<rgba 0 0.2977 1 1>}}
  User.ColorVariant {type=NiagaraInt32}
  User.ColorVariant_old {type=NiagaraBool}
  User.CustomColorIntensity {type=NiagaraFloat, defaultValue=1, description=""}
  User.RibbonIDOffset {type=NiagaraInt32}
  User.SkeletalMesh {
   type=NiagaraDataInterfaceSkeletalMesh
   defaultValue {
    struct=NiagaraExt_VariableValue_DataInterface
    value {dataInterface=&self:NiagaraDataInterfaceSkeletalMesh_2}
   }
  }
  User.TrailOpacity {type=NiagaraFloat, defaultValue=0.5, description=""}
  User.UseCustomColor {type=NiagaraFloat, defaultValue=0, description=""}
 ]
}
